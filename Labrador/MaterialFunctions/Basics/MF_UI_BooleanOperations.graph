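# MF_UI_BooleanOperations is a library-exposed material function that compares two scalars, A and B, and outputs 1 or 0 as true or false.
# Or and And derive from a saturated sum and absolute product, with OneMinus giving their negations.
# Greater, less, equal and or-equal outputs derive from a shared A minus B reroute.
# Saturate, Abs, Ceil and Floor turn that difference into binary results without branching.
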MaterialFunction /Labrador/MaterialFunctions/Basics/MF_UI_BooleanOperations
graph {
 properties {bExposeToLibrary=true, description="Compares two scalars and returns 1 or 0 (true or false)"}
 nodes [
  n1 = Saturate(None=Abs(None=NamedRerouteUsage()))
  FunctionOutput(None=OneMinus(None=Ceil(None=@n1))) {description=Returns 1 if equal and 0 if not, outputName=IsEqual, sortPriority=8}
  n2 = Saturate(None=Add(A=NamedRerouteUsage(), B=NamedRerouteUsage()))
  FunctionOutput(None=@n2) {description="A or B are true (1)", outputName=Or}
  FunctionOutput(None=OneMinus(None=@n2)) {description="Not (A or B are true (1))", outputName=Not Or, sortPriority=1}
  n3 = NamedRerouteUsage()
  n4 = NamedRerouteUsage()
  n5 = Saturate(None=Abs(None=Multiply(A=@n3, B=@n4)))
  FunctionOutput(None=@n5) {description="A AND B are true (1)", outputName=And, sortPriority=2}
  FunctionOutput(None=OneMinus(None=@n5)) {description="Not (A AND B are true (1))", outputName=Not And, sortPriority=3}
  n6 = NamedRerouteUsage()
  n7 = Ceil(None=Saturate(None=@n6))
  FunctionOutput(None=@n7) {description=A is greater than B, outputName=IsGreater, sortPriority=4}
  n8 = Abs(None=@n6)
  n9 = Saturate(None=Floor(None=OneMinus(None=@n8)))
  FunctionOutput(None=Add(A=@n7, B=@n9)) {description=A is greater or equal than B, outputName=IsGreaterOrEqual, sortPriority=5}
  n10 = NamedRerouteUsage()
  n11 = Ceil(None=Saturate(None=Multiply(A=@n10) {constB=-1}))
  FunctionOutput(None=@n11) {description=A is less than B, outputName=IsLess, sortPriority=6}
  FunctionOutput(None=Add(A=@n11, B=@n9)) {description=A is less or Equal than B, outputName=IsLessOrEqual, sortPriority=7}
  NamedRerouteDeclaration(None=Subtract(A=NamedRerouteUsage(), B=NamedRerouteUsage())) {name=SubtractAB, nodeColor=<rgba 0 0 0 1>}
  n12 = FunctionInput() {bUsePreviewValueAsDefault=true, description=First Scalar, inputName=A, inputType=FunctionInput_Scalar, previewValue=<1 0 0 1>}
  NamedRerouteDeclaration(None=@n12) {name=A, nodeColor=<rgba 0 0 0 1>}
  n13 = FunctionInput() {
   bUsePreviewValueAsDefault=true
   description=SecondScalar
   inputName=B
   inputType=FunctionInput_Scalar
   sortPriority=1
   previewValue=<1 0 0 1>
  }
  NamedRerouteDeclaration(None=@n13) {name=B, nodeColor=<rgba 0 0 0 1>}
 ]
}
